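# Unlit translucent material for Niagara sprites that draws soft, procedural smoke puffs.
# An upward- or downward-panning blurred noise texture offsets the UVs of an exponential radial gradient with adjustable radius and density, giving each puff a wobbling shape.
# Particle color feeds an emissive adjustment function, and particle alpha scales the gradient.
# A shading path switch applies depth fade on default and deferred paths, skipping it on forward and mobile.
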Material /BRCosmetics/Athena/Effects/Niagara/Smoke/M_Soft_Smoke
graph {
 properties {bUsedWithNiagaraSprites=true, blendMode=BLEND_Translucent, pixelDepthOffsetMode=PDOM_Legacy, shadingModel=MSM_Unlit, usedShadingModels=Unlit}
 nodes [
  n1 = ParticleColor()
  MaterialFunctionCall(Emissive=@n1.RGB) {materialFunction=&MF_FX_EmissiveAdjustments}
  n2 = TextureCoordinate()
  n3 = TextureCoordinate()
  n4 = Multiply(A=TextureSample(UVs=Panner(Coordinate=@n3) {speedY=0.5}) {samplerType=SAMPLERTYPE_Grayscale, texture=&LowResBlurredNoise}.R) {constB=0.3}
  n5 = Multiply(
   A=@n1.A
   B=MaterialFunctionCall(
    UVs=Add(A=@n2, B=@n4)
    Radius=ScalarParameter() {defaultValue=0.3, parameterName="Radius (S)"}
    Density=ScalarParameter() {defaultValue=1, parameterName="Density (S)"}
   ) {materialFunction=&RadialGradientExponential}.RadialGradientExponential
  )
  n6 = DepthFade(Opacity=@n5, FadeDistance=ScalarParameter() {defaultValue=10, parameterName=FadeDistance})
  ShadingPathSwitch(Default=@n6, Deferred=@n6, Forward=@n5, Mobile=@n5)
 ]
}
